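; Base Athena quest object Blueprint that gives child quest props shared hooks into a consolidated quest component.
; It forwards component events for quest readiness, success, custom pass/fail visuals and objective or quest updates to empty custom Handle events for children to implement.
; Interaction checks cast the pawn's controller to the Athena player controller and defer to the component's IsInteractable.
; ShouldDie always returns false, and comments direct base-class changes through named owners.
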(event OnQuestNotReady(QuestComponent)
  (params (PlayerController :type Fort%20Player%20Controller%20Object%20Reference))
  ;[1 1 1] If you are going to add or change anything in this base calss PLEASE talk to @Jon.Lietz, @Ken.Klopp or @Nicholas.Lacelle first!
  ;;[0 0.1344 1] Use to set up intiale state for when the user doesnt have the quest yet, and then once they have the quest
  (CallFunction|HandleQuestNotReady PlayerController))

(event Custom|HandleQuestNotReady
  (params (Player :type Fort%20Player%20Controller%20Object%20Reference)))

(event OnQuestReady(QuestComponent)
  (params (PlayerController :type Fort%20Player%20Controller%20Object%20Reference))
  ;[1 1 1] If you are going to add or change anything in this base calss PLEASE talk to @Jon.Lietz, @Ken.Klopp or @Nicholas.Lacelle first!
  ;;[0 0.1344 1] Use to set up intiale state for when the user doesnt have the quest yet, and then once they have the quest
  (CallFunction|HandleQuestReady PlayerController))

(event Custom|HandleQuestReady
  (params (Player :type Fort%20Player%20Controller%20Object%20Reference)))

(event Custom|HandleQuestComponentSuccessEvent
  (params (Player :type Fort%20Player%20Controller%20Object%20Reference)))

(event OnQuestComponentSuccessEvent(QuestComponent)
  (params (PlayerController :type Fort%20Player%20Controller%20Object%20Reference))
  ;[1 1 1] If you are going to add or change anything in this base calss PLEASE talk to @Jon.Lietz, @Ken.Klopp or @Nicholas.Lacelle first!
  ;;[0.8924 0.1314 1] Handle when the objective or quest this component is listening for is completed
  (CallFunction|HandleQuestComponentSuccessEvent PlayerController))

(event OnCustomSetVisualstoPassed(QuestComponent)
  (params (PlayerController :type Fort%20Player%20Controller%20Object%20Reference) (bFullyInitialized :type Boolean))
  ;[1 1 1] If you are going to add or change anything in this base calss PLEASE talk to @Jon.Lietz, @Ken.Klopp or @Nicholas.Lacelle first!
  ;;[1 1 1] This is for the BP to fully handle the custome visual respnse to the Conslidated quest component, these are only called when the component respionse is set to custom.
  ;;;[0 1 0.1127] Passed means we passed our check for the objective or quest being completed and this should be in the completed visual state. Fully Initalized true mean this was active during gameplay and the user JUST completed it. False means this is the fiirst time going through and we are inializing this to the completed state
  (CallFunction|HandleCustomSetVisualstoPassed PlayerController bFullyInitialized))

(event Custom|HandleCustomSetVisualsToPassed
  (params (Player :type Fort%20Player%20Controller%20Object%20Reference) (Inialized :type Boolean)))

(event Custom|HandleCustomSetVisualsToFailed
  (params (Player :type Fort%20Player%20Controller%20Object%20Reference)))

(event OnCustomSetVisualstoFailed(QuestComponent)
  (params (PlayerController :type Fort%20Player%20Controller%20Object%20Reference))
  ;[1 1 1] If you are going to add or change anything in this base calss PLEASE talk to @Jon.Lietz, @Ken.Klopp or @Nicholas.Lacelle first!
  ;;[1 1 1] This is for the BP to fully handle the custome visual respnse to the Conslidated quest component, these are only called when the component respionse is set to custom.
  ;;;[1 0 0] failed means we did not pass our check of the objective being completed or the quest being completed and this should be in the active visual state
  (CallFunction|HandleCustomSetVisualstoFailed PlayerController))

(event OnCustomUpdateObjective(QuestComponent)
  (params (PlayerController :type Fort%20Player%20Controller%20Object%20Reference) (AchievedCount :type Integer) (RequiredCount :type Integer))
  ;[1 1 1] If you are going to add or change anything in this base calss PLEASE talk to @Jon.Lietz, @Ken.Klopp or @Nicholas.Lacelle first!
  ;;[1 0.8818 0] this is called when the quest compoenet update type is set to Objective Updated. This means this BP call is 100% resposible for the visual state and will only recive updates when the objective count changes.
  (CallFunction|HandleCustomUpdateObjective PlayerController AchievedCount RequiredCount))

(event Custom|HandleCustomUpdateObjective
  (params (Player :type Fort%20Player%20Controller%20Object%20Reference) (Achieved%20Count :type Integer) (Goal%20Count :type Integer)))

(event OnUpdateQuest(QuestComponent)
  (params (PlayerController :type Fort%20Player%20Controller%20Object%20Reference))
  ;[1 1 1] If you are going to add or change anything in this base calss PLEASE talk to @Jon.Lietz, @Ken.Klopp or @Nicholas.Lacelle first!
  ;;[1 0.8818 0] this is called when the quest compoenet update type is set to Quest Updated. This means this BP call is 100% resposible for the visual state and will only recive updates when the quest achived count changes.
  (CallFunction|HandleQuestUpdate PlayerController))

(event Custom|Handle%20Quest%20Update
  (params (Player :type Fort%20Player%20Controller%20Object%20Reference)))

(fn BlueprintCanInteract
  (params (InteractingPawn :type Fort%20Pawn%20Object%20Reference) (InteractionBeingAttempted :type EInteractionBeingAttempted%20Enum :default "FirstInteraction") (InteractionType :type TInteractionType%20Enum :default "IT_NoInteraction"))
  (bind _asfort_player_controller_athena (Utilities|Casting|CastToFortPlayerControllerAthena (Pawn|GetController InteractingPawn))
    (:then
      (bind _output_get (LocalVariables|SetLocalFortPlayerController _asfort_player_controller_athena)))
    (:CastFailed
      _))
  (bind _returnvalue (Class|ConsolidatedQuestComponent|IsInteractable (Variables|Default|GetQuestComponent) _output_get))
  (return _returnvalue))

(fn ShouldDie
  (params (Damage :type Float%20%28single-precision%29) (EventInstigator :type Controller%20Object%20Reference) (DamageCauser :type Actor%20Object%20Reference) (bIsTest :type Boolean) (EffectContext :type Gameplay%20Effect%20Context%20Handle%20Structure%20%28by%20ref%29))
  (return false))

(fn ConstructionScript)
