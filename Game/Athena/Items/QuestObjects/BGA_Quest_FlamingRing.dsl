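; Flaming ring quest object that credits players who pass through it in a vehicle with a required gameplay tag.
; On server overlap it finds the vehicle, directly or through the player pawn, and sends a quest stat event for every passenger.
; QuestComponent ready events toggle its visibility per player controller, starting hidden, along with the looping spin audio.
; Success plays cosmetic sound, Niagara and shake feedback, then fades the loop.
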(event EventBeginPlay
  (for _array_element (Fortnite|GetLocalFortPlayerControllers true)
    (PlayerController|SetCompHiddenControllerOnly _array_element (Variables|Default|GetQuestComponent) true true)))

(event Collision|EventActorBeginOverlap
  (params (OtherActor :type Actor%20Object%20Reference))
  (bind _vehicleinterface (Variables|Default|GetVehicleInterface))
  ;[1 1 1] Ensure the player(s) are in an appropriate vehicle (meatball, valet, hoagie)
  (Networking|SwitchHasAuthority
    (:Authority
      (bind _asfort_vehicle_interface (Utilities|Casting|CastToFortVehicleInterface OtherActor)
        (:then
          (Variables|Default|SetVehicleInterface _asfort_vehicle_interface))
        (:CastFailed
          (bind _asfort_player_pawn_athena (Utilities|Casting|CastToFortPlayerPawnAthena OtherActor)) ; If hoagie, need the vehicle interface from the player
          (Variables|Default|SetVehicleInterface (Vehicle|GetVehicleInterface _asfort_player_pawn_athena))))
      (if (GameplayTags|HasAnyTags (Class|FortVehicleInterface|GetGameplayTags (Variables|Default|GetVehicleInterface)) (Variables|Default|GetRequiredVehicleTags))
        ;[1 1 1] Get all the players in the vehicle and check if the quest objective has not been completed
        (if (Vehicle|ContainsPlayers _vehicleinterface)
          (for _array_element (Vehicle|GetAllPassengers _vehicleinterface)
            (Utilities|IsValid _array_element
              (:Is%20Valid
                ;[1 1 1] Get Quest manager on each player and send stat event on quest completion
                (Utilities|IsValid (Fortnite|GetFortPlayerControllerfromActor _array_element)
                  (:Is%20Valid
                    (Quest|SendComplexCustomStatEvent (Quests|GetQuestManager (Fortnite|GetFortPlayerControllerfromActor _array_element) "Invalid") 0 (Variables|Default|GetT_Ring) "(GameplayTags=)" 1)))))))))))

(event OnQuestComponentSuccessEvent(QuestComponent)
  (params (PlayerController :type Fort%20Player%20Controller%20Object%20Reference))
  (Networking|CanExecuteCosmeticEvents
    (:True
      (Class|AthenaQuestFunctionLibrary|AthenaPlayQuestRewardFeedback false (Variables|Default|GetSuccessSound) (Variables|Default|GetSuccessNiagara) (Variables|Default|GetSuccessShakeFeedback) self (Game|Player|GetPlayerPawn PlayerController))
      (CallFunction|UpdateLoopingAudio))))

(event OnQuestNotReady(QuestComponent)
  (params (PlayerController :type Fort%20Player%20Controller%20Object%20Reference))
  (Variables|Default|SetQuestReady false)
  (Audio|Components|Audio|Stop (Variables|Default|GetSpinAudioLoop))
  (bind _returnvalue (PlayerController|SetCompHiddenControllerOnly PlayerController (Variables|Default|GetQuestComponent) true true))
  (if _returnvalue
    _))

(event OnQuestReady(QuestComponent)
  (params (PlayerController :type Fort%20Player%20Controller%20Object%20Reference))
  (Variables|Default|SetQuestReady true)
  (bind _returnvalue (PlayerController|SetCompHiddenControllerOnly PlayerController (Variables|Default|GetQuestComponent) false true))
  (if _returnvalue
    (Audio|Components|Audio|Play (Variables|Default|GetSpinAudioLoop) 0.0)))

(event EventDestroyed
  (CallFunction|UpdateLoopingAudio))

(fn UpdateLoopingAudio
  (if (Audio|Components|Audio|IsPlaying (Variables|Default|GetSpinAudioLoop))
    (Audio|Components|Audio|FadeOut (Variables|Default|GetSpinAudioLoop) 0.5 0.0 "Linear" :!class /Script/Engine.AudioComponent)))

(fn ConstructionScript
  (|CallParentFunction)
  (Variables|QuestStuff|SetQuestBackendObjectiveName (Variables|/Script/FortniteGame.ConsolidatedQuestComponent|GetObjectiveBackendName (Variables|Default|GetQuestComponent))))
