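; Quest ring actor for Athena that reports quest progress when a player pawn overlaps it.
; On overlap it stores the pawn, flushes net dormancy and sends a ring-tagged stat event to that player's quest manager, without checking the ready flag.
; Quest component events set that flag, toggle the component's visibility per controller, and play cosmetic-only success sound, Niagara and shake feedback.
; The looping-audio function is empty.
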(event Collision|EventActorBeginOverlap
  (params (OtherActor :type Actor%20Object%20Reference))
  (bind _asfort_player_pawn_athena (Utilities|Casting|CastToFortPlayerPawnAthena OtherActor))
  (bind _output_get (Variables|Default|SetPlayerPawn _asfort_player_pawn_athena))
  (Utilities|IsValid (Fortnite|GetFortPlayerControllerfromActor _output_get)
    (:Is%20Valid
      (Networking|FlushNetDormancy)
      (Quest|SendComplexCustomStatEvent (Quests|GetQuestManager (Fortnite|GetFortPlayerControllerfromActor _output_get) "Invalid") 0 (Variables|Default|GetT_Ring) "(GameplayTags=)" 1))))

(event OnQuestComponentSuccessEvent(QuestComponent)
  (params (PlayerController :type Fort%20Player%20Controller%20Object%20Reference))
  (Networking|CanExecuteCosmeticEvents
    (:True
      (Class|AthenaQuestFunctionLibrary|AthenaPlayQuestRewardFeedback false (Variables|Default|GetSuccessSound) (Variables|Default|GetSuccessNiagara) (Variables|Default|GetSuccessShakeFeedback) self (Variables|Default|GetPlayerPawn)))))

(event OnQuestNotReady(QuestComponent)
  (params (PlayerController :type Fort%20Player%20Controller%20Object%20Reference))
  (Variables|Default|SetQuestReady false)
  (PlayerController|SetCompHiddenControllerOnly PlayerController (Variables|Default|GetQuestComponent) true true))

(event OnQuestReady(QuestComponent)
  (params (PlayerController :type Fort%20Player%20Controller%20Object%20Reference))
  (Variables|Default|SetQuestReady true)
  (PlayerController|SetCompHiddenControllerOnly PlayerController (Variables|Default|GetQuestComponent) false true))

(fn UpdateLoopingAudio)

(fn ConstructionScript
  (|CallParentFunction))
